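# Translucent unlit Niagara sprite and mesh particle material that, as its name suggests, draws a LimeRock ripple wave effect.
# Panned radial UVs, distorted by noise, drive three offset plasma samples for RGB split, plus a base texture, optionally blended into particle color.
# Noise-warped sphere masks carve a ring in opacity, which gets depth fade (Default and Deferred paths only), view-angle fade and optional edge masking.
# A camera-offset function drives WPO.
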Material /LimeRock/Assets/FX/Resources/M_LimeRockRipple_Wave
graph {
 properties {
  bUsedWithNiagaraMeshParticles=true
  bUsedWithNiagaraSprites=true
  blendMode=BLEND_Translucent
  pixelDepthOffsetMode=PDOM_Legacy
  shadingModel=MSM_Unlit
  usedShadingModels=Unlit
 }
 nodes [
  Constant() {r=1}
  Constant() {r=1}
  n1 = ParticleColor()
  n2 = MaterialFunctionCall() {materialFunction=&VectorToRadialValue}
  n3 = ComponentMask(None=VectorParameter() {parameterName=Speed, defaultValue=<rgba 0 -0.4 0 0>}.RGB) {g=true, r=true}
  n4 = Panner(Coordinate=MaterialFunctionCall() {materialFunction=&VectorToRadialValue}.Radial Coordinates) {speedX=0.2, speedY=-1}
  n5 = ComponentMask(None=VectorParameter() {parameterName=noise speed, defaultValue=<rgba 1 1 0 1>}.RGB) {g=true, r=true}
  n6 = Multiply(
   A=TextureSample(UVs=Multiply(A=@n4, B=@n5)) {samplerType=SAMPLERTYPE_LinearGrayscale, texture=/Game/Effects/Fort_Effects/Emotes/UltraLight/Texture/T_DirectionalNoise_Greyscale}.R
   B=ScalarParameter() {defaultValue=0.2, parameterName=Noise_Intensity}
  ) {constB=0.2}
  n7 = Add(A=Multiply(A=Panner(Coordinate=@n2.Radial Coordinates, Speed=@n3) {speedY=-0.4}, B=Constant2Vector() {g=1, r=1}), B=@n6)
  n8 = ScalarParameter() {defaultValue=0.02, parameterName=Color offset}
  n9 = TextureSampleParameter2D(UVs=Add(A=@n7, B=@n8)) {parameterName=Tex Noise, samplerType=SAMPLERTYPE_LinearGrayscale, texture=/Game/Effects/Fort_Effects/Emotes/UltraLight/Texture/T_PlasmaCoils}
  n10 = TextureSampleParameter2D(UVs=@n7) {parameterName=Tex Noise, samplerType=SAMPLERTYPE_LinearGrayscale, texture=&T_PlasmaCoils}
  n11 = TextureSampleParameter2D(UVs=Subtract(A=@n7, B=@n8)) {parameterName=Tex Noise, samplerType=SAMPLERTYPE_LinearGrayscale, texture=&T_PlasmaCoils}
  n12 = MaterialFunctionCall(X=@n9.R, Y=@n10.R, Z=@n11.R) {materialFunction=&MakeFloat3}
  n13 = TextureSampleParameter2D(UVs=@n7) {
   parameterName=Tex Base
   samplerType=SAMPLERTYPE_LinearGrayscale
   texture=/Game/Characters/Player/Male/Medium/Bodies/M_MED_Citadel/FX/T_Citadel_Noise_Dissolve_Handpaint
  }
  n14 = Multiply(A=@n13.R, B=ScalarParameter() {defaultValue=0.7, parameterName=Center Alpha}) {constB=0.7}
  n15 = Add(A=Desaturation(None=Power(Base=@n12.Result) {constExponent=3}, Fraction=ScalarParameter() {defaultValue=-0.2, parameterName=Des}), B=@n14)
  MaterialFunctionCall(Emissive=LinearInterpolate(A=@n1.RGB, B=Multiply(A=@n15, B=@n1.RGB), Alpha=ScalarParameter() {parameterName=Use Colorful Texture})) {materialFunction=&MF_FX_EmissiveAdjustments}
  MaterialFunctionCall!(WPO, OffsetPercent=ScalarParameter() {defaultValue=0.15, parameterName=OffsetP}) {materialFunction=&/MF_CameraOffsetButKeepSize}
  n16 = Power(Base=@n10.R, Exp=ScalarParameter() {defaultValue=3, parameterName=Alpha power}) {constExponent=3}
  n17 = SphereMask(A=TextureCoordinate(), B=Constant() {r=0.5}) {attenuationRadius=0.5, hardnessPercent=0}
  n18 = Power(Base=Add(A=Add(A=@n16, B=@n14)) {constB=0}, Exp=LinearInterpolate(Alpha=OneMinus(None=@n17)) {constA=-0.1, constB=2})
  n19 = TextureCoordinate()
  n20 = Constant2Vector() {g=-0.07, r=-0.1}
  n21 = SphereMask(A=Add(A=Add(A=@n19, B=@n20), B=@n6), B=Constant() {r=0.5}) {attenuationRadius=0.45, hardnessPercent=0}
  n22 = MaterialFunctionCall() {materialFunction=&VectorToRadialValue}
  n23 = TextureSample(UVs=Multiply(A=Panner(Coordinate=@n22.Radial Coordinates) {speedX=0.2, speedY=1}, B=Constant2Vector() {g=0.5, r=1})) {samplerType=SAMPLERTYPE_LinearGrayscale, texture=&T_SimplePerlin_Greyscale}
  n24 = Add(A=Add(A=TextureCoordinate(), B=Constant2Vector() {r=-0.05}), B=Multiply(A=Multiply(A=@n23.R) {constB=0.3}) {constB=0.2})
  n25 = Constant() {r=0.5}
  n26 = Power(Base=OneMinus(None=SphereMask(A=@n24, B=@n25) {attenuationRadius=0.3, hardnessPercent=30}))
  n27 = Multiply(A=@n1.A, B=Multiply(A=Power(Base=@n18) {constExponent=1.2}, B=Multiply(A=@n21, B=@n26)))
  n28 = DepthFade(Opacity=@n27, FadeDistance=ScalarParameter() {defaultValue=10, parameterName=FadeDistance}) {fadeDistanceDefault=10}
  n29 = ScalarParameter() {parameterName="FadeSensitivity (optional) (S)"}
  n30 = Multiply(A=SphereMask(A=TextureCoordinate(), B=Constant() {r=0.5}) {attenuationRadius=0.5, hardnessPercent=0}) {constB=5}
  Multiply(
   A=Multiply(
    A=ShadingPathSwitch(Default=@n28, Deferred=@n28, Forward=@n27, Mobile=@n27)
    B=MaterialFunctionCall("FadeSensitivity (optional)"=@n29) {materialFunction=&FadeBasedOnViewAngle}.Result
   )
   B=LinearInterpolate(B=Saturate(None=@n30), Alpha=ScalarParameter() {parameterName=Use sphere edge mask}) {constA=1}
  )
 ]
}
