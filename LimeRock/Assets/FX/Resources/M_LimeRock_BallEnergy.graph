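# Translucent unlit material for an energy ball effect, flagged for Niagara mesh particles.
# A plasma texture pans over UVs with a power-curved V axis, then passes through refine and V edge-mask functions.
# That result is scaled by particle alpha and depth fade to form opacity, with an optional sine-based cone mask across U.
# Particle color drives scaled emissive, and a camera-offset function that keeps apparent size supplies WPO.
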Material /LimeRock/Assets/FX/Resources/M_LimeRock_BallEnergy
graph {
 properties {bUsedWithNiagaraMeshParticles=true, blendMode=BLEND_Translucent, shadingModel=MSM_Unlit, usedShadingModels=Unlit}
 nodes [
  Constant2Vector() {g=1, r=1}
  n1 = ParticleColor()
  n2 = MaterialFunctionCall() {materialFunction=&LinearGradient}
  n3 = Power(Base=@n2.VGradient) {constExponent=1.5}
  n4 = VectorParameter() {parameterName=Param_1, defaultValue=<rgba 1 0.7 0 1>}
  n5 = Panner(
   Coordinate=Multiply(A=AppendVector(A=@n2.UGradient, B=@n3), B=AppendVector(A=@n4.R, B=@n4.R))
   Time=Multiply(A=Time(), B=ScalarParameter() {defaultValue=1, parameterName=PanSpeed})
  ) {speedY=-0.4}
  n6 = VectorParameter() {defaultValue=<rgba 4 3 0.5 0.5>}
  n7 = VectorParameter() {
   group=Grad
   parameterName=Grad_EdgeMask_InvertControl
   channelNames {a="", b=Power, g=Multiply, r=Offset}
   defaultValue=<rgba 0.0968 1 25.8425 1>
  }
  n8 = VectorParameter() {group=Grad, parameterName=Grad_EdgeMask_Control, channelNames {a="", b=Power, g=Multiply, r=Offset}, defaultValue=<rgba 0 2 1 1>}
  n9 = Multiply(
   A=MaterialFunctionCall(
    Texture=TextureSample(UVs=@n5) {samplerType=SAMPLERTYPE_LinearGrayscale, texture=/Game/Effects/Fort_Effects/Emotes/UltraLight/Texture/T_Plasma_A_Greyscale}.R
    RGB=@n6.RGB
    A=@n6.RGB
   ) {materialFunction=&MF_FX_Refine}.Result
   B=MaterialFunctionCall!(UorV=@n2.VGradient, Invert_EdgeControl=@n7, NoInvert_EdgeControl=@n8) {materialFunction=/Game/Effects/Niagara/MasterMaterials/Functions/Common/FXCommon_EdgeMask_Ex}.Mask
  )
  n10 = DepthFade(FadeDistance=ScalarParameter() {defaultValue=500, parameterName=sDF})
  n11 = Multiply(A=Multiply(A=Multiply(A=@n1.A, B=@n9), B=Power(Base=@n10)), B=ScalarParameter() {defaultValue=5, parameterName=ScaleA})
  n12 = Saturate(None=Sine(None=@n2.UGradient))
  n13 = ScalarParameter() {defaultValue=3, parameterName=Mask3}
  LinearInterpolate(A=@n11, B=Multiply(A=Power(Base=@n12, Exp=@n13), B=@n11), Alpha=ScalarParameter() {parameterName=UseConeMask}) {constAlpha=1}
  MaterialFunctionCall!(WPO, OffsetPercent=ScalarParameter() {defaultValue=0.1, parameterName=CameraOffset}) {materialFunction=&/MF_CameraOffsetButKeepSize}
  Multiply(
   A=ScalarParameter() {defaultValue=0.2, parameterName=ScaleC}
   B=MaterialFunctionCall(Emissive=@n1.RGB) {materialFunction=&MF_FX_EmissiveAdjustments}.Result
  )
 ]
}
